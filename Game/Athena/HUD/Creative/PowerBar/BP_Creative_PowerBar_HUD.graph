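# Creative power bar HUD widget that shows the owning player's power level on a material progress bar, hidden by default.
# An async gameplay tag listener waits for Creative.PowerBar.Show, then reveals the bar, caches the player's CreativePowerBarComponent and starts a looping timer.
# The timer sets progress from the component's power percentage.
# A tag-removed listener, registered on each add, hides the bar, clears the timer and resets progress.
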WidgetBlueprint /Game/Athena/HUD/Creative/PowerBar/BP_Creative_PowerBar_HUD
graphs [
 EventGraph {nodes [
   Event OnInitialized()
   > SetVisibility(self=Get SimpleMaterialProgressBar() -> (SimpleMaterialProgressBar:Simple Material Progress Bar Object Reference), InVisibility=Hidden)
   n1 = GetOwningPlayer()
   > n2 = WaitGameplayTagAddToActor!(
    execute:Exec
    TargetActor:Actor Object Reference=@n1
    Tag:Gameplay Tag Structure=Get ShowTag() -> (ShowTag:Gameplay Tag Structure)
    OnlyTriggerOnce:bool=false
   ) -> (then:Exec, AsyncTaskProxy:Ability Async Wait Gameplay Tag Added Object Reference, Added:Exec) {is=title K2Node_AsyncAction}
   n3 = Sequence(execute=@n2.Added)
   > SetVisibility(self=Get SimpleMaterialProgressBar(), InVisibility=HitTestInvisible)
   > n4 = Set PlayerPowerBarComponent(PlayerPowerBarComponent:Creative Power Bar Component Object Reference=GetComponentByClass!(self:Actor Object Reference=@n1, ComponentClass:Actor Component Class Reference=/Script/FortniteGame.CreativePowerBarComponent) -> (ReturnValue:Creative Power Bar Component Object Reference)) -> (Output_Get:Creative Power Bar Component Object Reference)
   > Is Valid(InputObject=@n4)
   > n5 = Set Timer by Function Name!(
    execute
    Object:Object Reference
    FunctionName:String=UpdateBarValue
    Time:float=Get BarUpdateRate() -> (BarUpdateRate:double)
    bLooping:bool=true
    bMaxOncePerFrame:bool=false
    InitialStartDelay:float=0
    InitialStartDelayVariance:float=0
   ) -> (then, ReturnValue:Timer Handle Structure) {is=title K2Node_CallFunction}
   > Set BarUpdateTimerHandle(BarUpdateTimerHandle:Timer Handle Structure=@n5) -> (Output_Get:Timer Handle Structure)
   n6 = WaitGameplayTagRemoveFromActor!(execute:Exec=@n3.then_1, TargetActor:Actor Object Reference=@n1, Tag:Gameplay Tag Structure=Get ShowTag(), OnlyTriggerOnce:bool=false) -> (then:Exec, AsyncTaskProxy:Ability Async Wait Gameplay Tag Removed Object Reference, Removed:Exec) {is=title K2Node_AsyncAction}
   SetVisibility(execute=@n6.Removed, self=Get SimpleMaterialProgressBar(), InVisibility=Hidden)
   > Clear and Invalidate Timer by Handle(Handle=Get BarUpdateTimerHandle() -> (BarUpdateTimerHandle:Timer Handle Structure))
   > SetProgress!(execute, self:Fort Simple Material Progress Bar Object Reference=Get SimpleMaterialProgressBar(), Progress:float=0) -> (then)
  ]}
 UpdateBarValue {nodes [
   UpdateBarValue() {is=title K2Node_FunctionEntry}
   > n1 = Get!(execute:Exec) -> (then:Exec, else:Exec, PlayerPowerBarComponent:Creative Power Bar Component Object Reference)
   > SetProgress(
    self=Get SimpleMaterialProgressBar()
    Progress=GetPowerPercentage!(self:Creative Power Bar Component Object Reference=@n1) -> (ReturnValue:float)
   )
  ]}
]
variables [
 BarUpdateRate {type=number, default=0.05}
 BarUpdateTimerHandle {type=TimerHandle}
 PlayerPowerBarComponent {type=CreativePowerBarComponent, default=null}
 ShowTag {type=GameplayTag, default {tagName=Creative.PowerBar.Show}}
]
widgets {
 info {inheritedWidgetCount=0, parentClass=&UserWidget, rootWidgetClass=&ScaleBox, widgetCount=3}
 widgets [
  ScaleBox_139 {widgetClassPath=&ScaleBox, widget=ScaleBox_139, properties {stretch=UserSpecified}}
  SizeBox_175 {
   widgetClassPath=&SizeBox
   parent=ScaleBox_139
   widget=SizeBox_175
   properties {bOverride_HeightOverride=true, bOverride_WidthOverride=true, heightOverride=24, widthOverride=450}
   slot {class=ScaleBoxSlot, properties {horizontalAlignment=HAlign_Fill, verticalAlignment=VAlign_Fill}}
  }
  SimpleMaterialProgressBar {
   widgetClassPath=/Game/Athena/HUD/SimpleMaterialProgressBar
   bIsVariable=true
   parent=SizeBox_175
   widget=SimpleMaterialProgressBar
   properties {bOverrideDefaultColorA=true, bOverrideDefaultColorB=true, cachedColorA=<rgba 1 0.9066 0 1>, cachedColorB=<rgba 1 0.0078 0 1>}
   slot {class=SizeBoxSlot}
  }
 ]
}
